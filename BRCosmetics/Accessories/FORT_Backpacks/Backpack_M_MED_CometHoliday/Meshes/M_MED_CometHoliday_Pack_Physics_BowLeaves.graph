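# Physics asset for the bow and leaves on the CometHoliday backpack cosmetic, providing secondary-motion bodies and joints.
# A large bell capsule is the root, branching through tiny base capsules into left and right two-segment bow chains and three-segment long and short leaf chains.
# The dynamic segments use tight swing limits with zero-degree twist limits, while the joints between base bodies leave most axes free.
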PhysicsAsset /BRCosmetics/Accessories/FORT_Backpacks/Backpack_M_MED_CometHoliday/Meshes/M_MED_CometHoliday_Pack_Physics_BowLeaves
bodies {
 base_bell [{shapeType=Capsule, length=8.5166, radius=7.209, shapeName=base_bell_capsule, center=<4.383 -0.53 -0.0139>, rotation=<rot 0 92.6341 -89.7291>}]
 base_bow [{shapeType=Capsule, length=0.505, radius=0.505, shapeName=base_bow_capsule, rotation=<rot 0 -90 0>}]
 base_bow_01_l [{shapeType=Capsule, length=0.505, radius=0.505, shapeName=base_bow_01_l_capsule, rotation=<rot 0 -90 0>}]
 base_bow_01_r [{shapeType=Capsule, length=0.505, radius=0.505, shapeName=base_bow_01_r_capsule, rotation=<rot 0 -90 0>}]
 base_leaves [{shapeType=Capsule, length=0.505, radius=0.505, shapeName=base_leaves_capsule, rotation=<rot 0 -90 0>}]
 base_longleaves [{shapeType=Capsule, length=0.505, radius=0.505, shapeName=base_longleaves_capsule, rotation=<rot 0 -90 0>}]
 base_shortleaves [{shapeType=Capsule, length=0.505, radius=0.505, shapeName=base_shortleaves_capsule, rotation=<rot 0 -90 0>}]
 dyn_bow_01_l [{shapeType=Capsule, length=2.6973, radius=2.0312, shapeName=dyn_bow_01_l_capsule, center=<1.1722 0 0>, rotation=<rot 0 0 90>}]
 dyn_bow_01_r [{shapeType=Capsule, length=2.4085, radius=2.0073, shapeName=dyn_bow_01_r_capsule, center=<1.1722 0 0>, rotation=<rot 0 0 90>}]
 dyn_bow_02_l [{shapeType=Capsule, length=3.0318, radius=2.0841, shapeName=dyn_bow_02_l_capsule, center=<1 0 0>, rotation=<rot 0 0 -90>}]
 dyn_bow_02_r [{shapeType=Capsule, length=3.2477, radius=2.069, shapeName=dyn_bow_02_r_capsule, center=<1 0 0>, rotation=<rot 0 0 90>}]
 dyn_longleaves_01 [{
   shapeType=Capsule
   length=3.3812
   radius=2
   shapeName=dyn_longleaves_01_capsule
   center=<3.3312 0.0445 0.154>
   rotation=<rot 0 86.2886 -86.4106>
  }]
 dyn_longleaves_02 [{shapeType=Capsule, length=4.4746, radius=2, shapeName=dyn_longleaves_02_capsule, center=<2.9904 0 0>, rotation=<rot -90 0 0>}]
 dyn_longleaves_03 [{shapeType=Capsule, length=4.3805, radius=2, shapeName=dyn_longleaves_03_capsule, center=<2.4611 0 0>, rotation=<rot -90 0 0>}]
 dyn_shortleaves_01 [{
   shapeType=Capsule
   length=3.1288
   radius=2
   shapeName=dyn_shortleaves_01_capsule
   center=<2.1108 0.1869 -0.3806>
   rotation=<rot 0 -92.1223 -83.9458>
  }]
 dyn_shortleaves_02 [{shapeType=Capsule, length=4.1264, radius=2, shapeName=dyn_shortleaves_02_capsule, center=<2.0776 0 0>, rotation=<rot 90 0 0>}]
 dyn_shortleaves_03 [{shapeType=Capsule, length=3.1034, radius=2, shapeName=dyn_shortleaves_03_capsule, center=<1.9539 0 0>, rotation=<rot 90 0 0>}]
}
constraints [
 {
  bone1Name=base_bow_01_l
  bone2Name=base_bow
  swing1LimitDegrees=45
  swing1Motion=Locked
  swing2LimitDegrees=45
  swing2Motion=Free
  twistLimitDegrees=45
  twistMotion=Free
 }
 {bone1Name=dyn_bow_01_l, bone2Name=base_bow_01_l, swing1LimitDegrees=10, swing2LimitDegrees=15, twistLimitDegrees=0}
 {bone1Name=dyn_bow_02_l, bone2Name=dyn_bow_01_l, swing1LimitDegrees=20, swing2LimitDegrees=15, twistLimitDegrees=0}
 {
  bone1Name=base_bow_01_r
  bone2Name=base_bow
  swing1LimitDegrees=45
  swing1Motion=Locked
  swing2LimitDegrees=45
  swing2Motion=Free
  twistLimitDegrees=45
  twistMotion=Free
 }
 {
  bone1Name=base_longleaves
  bone2Name=base_leaves
  swing1LimitDegrees=45
  swing1Motion=Free
  swing2LimitDegrees=45
  swing2Motion=Free
  twistLimitDegrees=45
  twistMotion=Free
 }
 {bone1Name=dyn_longleaves_01, bone2Name=base_longleaves, swing1LimitDegrees=15, swing2LimitDegrees=5, twistLimitDegrees=0}
 {bone1Name=dyn_longleaves_02, bone2Name=dyn_longleaves_01, swing1LimitDegrees=20, swing2LimitDegrees=5, twistLimitDegrees=0}
 {bone1Name=dyn_longleaves_03, bone2Name=dyn_longleaves_02, swing1LimitDegrees=20, swing2LimitDegrees=5, twistLimitDegrees=0}
 {
  bone1Name=base_shortleaves
  bone2Name=base_leaves
  swing1LimitDegrees=45
  swing1Motion=Locked
  swing2LimitDegrees=45
  swing2Motion=Free
  twistLimitDegrees=45
  twistMotion=Free
 }
 {bone1Name=dyn_shortleaves_01, bone2Name=base_shortleaves, swing1LimitDegrees=10, swing2LimitDegrees=5, twistLimitDegrees=0}
 {bone1Name=dyn_shortleaves_02, bone2Name=dyn_shortleaves_01, swing1LimitDegrees=20, swing2LimitDegrees=5, twistLimitDegrees=0}
 {bone1Name=dyn_shortleaves_03, bone2Name=dyn_shortleaves_02, swing1LimitDegrees=20, swing2LimitDegrees=5, twistLimitDegrees=0}
 {
  bone1Name=base_leaves
  bone2Name=base_bell
  swing1LimitDegrees=45
  swing1Motion=Locked
  swing2LimitDegrees=45
  swing2Motion=Free
  twistLimitDegrees=45
  twistMotion=Free
 }
 {
  bone1Name=base_bow
  bone2Name=base_bell
  swing1LimitDegrees=45
  swing1Motion=Locked
  swing2LimitDegrees=45
  swing2Motion=Free
  twistLimitDegrees=45
  twistMotion=Free
 }
 {bone1Name=dyn_bow_02_r, bone2Name=dyn_bow_01_r, swing1LimitDegrees=20, swing2LimitDegrees=15, twistLimitDegrees=0}
 {bone1Name=dyn_bow_01_r, bone2Name=base_bow_01_r, swing1LimitDegrees=20, swing2LimitDegrees=15, twistLimitDegrees=0}
]
