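# Physics asset for the ribbons on the CometHoliday holiday backpack cosmetic, using capsule bodies on the pack's bones.
# A large bell capsule is the root, with small bow and ribbon base bodies constrained to it, and left and right ribbon bases branching from the latter.
# Each side continues as a three-segment dynamic chain with narrow, partly asymmetric swing limits and zero-degree twist limits, while base joints mix free and locked axes.
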PhysicsAsset /BRCosmetics/Accessories/FORT_Backpacks/Backpack_M_MED_CometHoliday/Meshes/M_MED_CometHoliday_Pack_Physics_Ribbonds
bodies {
 base_bell [{shapeType=Capsule, length=8.5166, radius=7.209, shapeName=base_bell_capsule, center=<4.383 -0.53 -0.0139>, rotation=<rot 0 92.6341 -89.7291>}]
 base_bow [{shapeType=Capsule, length=0.505, radius=0.505, shapeName=base_bow_capsule, rotation=<rot 0 -90 0>}]
 base_ribbond [{shapeType=Capsule, length=0.505, radius=0.505, shapeName=base_ribbond_capsule, rotation=<rot 0 -90 0>}]
 base_ribbond_01_l [{shapeType=Capsule, length=0.505, radius=0.505, shapeName=base_ribbond_01_l_capsule, rotation=<rot 0 -90 0>}]
 base_ribbond_01_r [{shapeType=Capsule, length=0.505, radius=0.505, shapeName=base_ribbond_01_r_capsule, rotation=<rot 0 -90 0>}]
 dyn_ribbond_01_l [{shapeType=Capsule, length=1.2106, radius=1, shapeName=dyn_ribbond_01_l_capsule, center=<1 0 0>, rotation=<rot 90 0 0>}]
 dyn_ribbond_01_r [{shapeType=Capsule, length=1.2784, radius=1, shapeName=dyn_ribbond_01_r_capsule, center=<1 0 0>, rotation=<rot 90 0 0>}]
 dyn_ribbond_02_l [{shapeType=Capsule, length=2.352, radius=1, shapeName=dyn_ribbond_02_l_capsule, rotation=<rot 90 0 0>}]
 dyn_ribbond_02_r [{shapeType=Capsule, length=2.352, radius=1, shapeName=dyn_ribbond_02_r_capsule, rotation=<rot 90 0 0>}]
 dyn_ribbond_03_l [{shapeType=Capsule, length=1.7148, radius=1, shapeName=dyn_ribbond_03_l_capsule, center=<1 0 0>, rotation=<rot 90 0 0>}]
 dyn_ribbond_03_r [{shapeType=Capsule, length=1.7182, radius=1, shapeName=dyn_ribbond_03_r_capsule, rotation=<rot 90 0 0>}]
}
constraints [
 {
  bone1Name=base_ribbond_01_l
  bone2Name=base_ribbond
  swing1LimitDegrees=45
  swing1Motion=Free
  swing2LimitDegrees=45
  swing2Motion=Free
  twistLimitDegrees=45
  twistMotion=Free
 }
 {bone1Name=dyn_ribbond_01_l, bone2Name=base_ribbond_01_l, swing1LimitDegrees=10, swing2LimitDegrees=7, twistLimitDegrees=0}
 {bone1Name=dyn_ribbond_02_l, bone2Name=dyn_ribbond_01_l, swing1LimitDegrees=20, swing2LimitDegrees=20, twistLimitDegrees=0}
 {
  bone1Name=base_ribbond_01_r
  bone2Name=base_ribbond
  swing1LimitDegrees=45
  swing1Motion=Locked
  swing2LimitDegrees=45
  swing2Motion=Free
  twistLimitDegrees=45
  twistMotion=Free
 }
 {bone1Name=dyn_ribbond_01_r, bone2Name=base_ribbond_01_r, swing1LimitDegrees=5, swing2LimitDegrees=20, twistLimitDegrees=0}
 {bone1Name=dyn_ribbond_02_r, bone2Name=dyn_ribbond_01_r, swing1LimitDegrees=20, swing2LimitDegrees=25, twistLimitDegrees=0}
 {
  bone1Name=base_ribbond
  bone2Name=base_bell
  swing1LimitDegrees=45
  swing1Motion=Locked
  swing2LimitDegrees=45
  swing2Motion=Free
  twistLimitDegrees=45
  twistMotion=Free
 }
 {
  bone1Name=base_bow
  bone2Name=base_bell
  swing1LimitDegrees=45
  swing1Motion=Locked
  swing2LimitDegrees=45
  swing2Motion=Free
  twistLimitDegrees=45
  twistMotion=Free
 }
 {bone1Name=dyn_ribbond_03_l, bone2Name=dyn_ribbond_02_l, swing1LimitDegrees=20, swing2LimitDegrees=20, twistLimitDegrees=0}
 {bone1Name=dyn_ribbond_03_r, bone2Name=dyn_ribbond_02_r, swing1LimitDegrees=20, swing2LimitDegrees=20, twistLimitDegrees=0}
]
